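# Masked, Default Lit material that, by its name, provides the Item Shop filter background.
# MF_UI_BackgroundNebulous generates the backdrop from two close bright-blue colors, with reduced wind opacity and full triangle opacity.
# Its Final Color feeds the Vault World Color input of MF_UI_VaultWorld_FloorSetup, reusing the front-end item-preview vault floor setup.
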Material /ItemShop/UI/Art/Filter/M_UI_ShopFilter_BG
graph {
 properties {bAutomaticallySetUsageInEditor=false, bCastRayTracedShadows=false, blendMode=BLEND_Masked, usedShadingModels=Default Lit}
 nodes [MaterialFunctionCall!(Vault World Color=MaterialFunctionCall!(
    Primary Color=Constant3Vector() {constant=<rgba 0 0.3274 0.9823 1>}
    Secondary Color=Constant3Vector() {constant=<rgba 0 0.2423 0.9734 1>}
    Background Color
    Wind Opacity=Constant() {r=0.3}
    Rotation Multiplier
    Triangle Opacity=Constant() {r=1}
    Offset Multiplier
   ) {materialFunction=/FortUILibrary/Materials/Background/Nebulous/MF_UI_BackgroundNebulous}.Final Color) {materialFunction=/Game/Environments/FrontEnd/Blueprints/ItemPreview/VaultWorldTransitions/MaterialFunctions/MF_UI_VaultWorld_FloorSetup}]
}
